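# Widget prompt for the song menu, showing a keybind bound to IA_OpenQuestPanel (hidden on touch), a "Game Menu" label and an icon.
# PreConstruct collapses the label when HideMenuLabel is set.
# Construct starts async gameplay tag waits on the owning PlayerState, collapsing the widget when Pilgrim.Widget.SongMenuPrompt is added and showing it when removed.
# Destruct cancels both waits and clears their stored handles.
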WidgetBlueprint /SparksUICommon/Widgets/WBP_SongMenuPrompt
graphs [EventGraph {nodes [
   Event PreConstruct()
   > SetVisibility(
    self=Get MenuLabel() -> (MenuLabel:Common Text Object Reference)
    InVisibility=Select(Option 0=SelfHitTestInvisible, Option 1=Collapsed, Index=Get HideMenuLabel() -> (HideMenuLabel:bool))
   )
   Event Construct()
   > n1 = Sequence()
   n2 = Get PlayerState!(self:Controller Object Reference=GetOwningPlayer()) -> (PlayerState:Player State Object Reference)
   > n3 = WaitGameplayTagAddToActor!(
    execute:Exec
    TargetActor:Actor Object Reference=@n2
    Tag:Gameplay Tag Structure=(TagName="Pilgrim.Widget.SongMenuPrompt")
    OnlyTriggerOnce:bool=false
   ) -> (then:Exec, AsyncTaskProxy:Ability Async Wait Gameplay Tag Added Object Reference, Added:Exec) {is=title K2Node_AsyncAction}
   > Set SongMenuPromptTagAddedEventHandle(SongMenuPromptTagAddedEventHandle:Ability Async Wait Gameplay Tag Added Object Reference=@n3) -> (Output_Get:Ability Async Wait Gameplay Tag Added Object Reference)
   SetVisibility(execute=@n3.Added, InVisibility=Collapsed)
   n4 = WaitGameplayTagRemoveFromActor!(
    execute:Exec=@n1.then_1
    TargetActor:Actor Object Reference=@n2
    Tag:Gameplay Tag Structure=(TagName="Pilgrim.Widget.SongMenuPrompt")
    OnlyTriggerOnce:bool=false
   ) -> (then:Exec, AsyncTaskProxy:Ability Async Wait Gameplay Tag Removed Object Reference, Removed:Exec) {is=title K2Node_AsyncAction}
   > Set SongMenuPromptTagRemovedEventHandle(SongMenuPromptTagRemovedEventHandle:Ability Async Wait Gameplay Tag Removed Object Reference=@n4) -> (Output_Get:Ability Async Wait Gameplay Tag Removed Object Reference)
   SetVisibility(execute=@n4.Removed, InVisibility=SelfHitTestInvisible)
   Event Destruct()
   > n5 = Sequence()
   n6 = Get SongMenuPromptTagAddedEventHandle() -> (SongMenuPromptTagAddedEventHandle:Ability Async Wait Gameplay Tag Added Object Reference)
   > Is Valid(InputObject=@n6)
   > Cancel(self:Cancellable Async Action Object Reference=@n6)
   > Set SongMenuPromptTagAddedEventHandle()
   n7 = Get SongMenuPromptTagRemovedEventHandle() -> (SongMenuPromptTagRemovedEventHandle:Ability Async Wait Gameplay Tag Removed Object Reference)
   Is Valid(exec=@n5.then_1, InputObject=@n7)
   > Cancel(self=@n7)
   > Set SongMenuPromptTagRemovedEventHandle()
   Event Tick()
  ]}]
variables [
 HideMenuLabel {type=boolean}
 SongMenuPromptTagAddedEventHandle {type=AbilityAsync_WaitGameplayTagAdded, default=null}
 SongMenuPromptTagRemovedEventHandle {type=AbilityAsync_WaitGameplayTagRemoved, default=null}
]
widgets {
 info {inheritedWidgetCount=0, parentClass=&UserWidget, rootWidgetClass=&Overlay, widgetCount=8}
 widgets [
  SongMenuPrompt {widgetClassPath=&Overlay, bIsVariable=true, widget=SongMenuPrompt}
  BackingPanel {
   widgetClassPath=&Image
   bIsVariable=true
   parent=SongMenuPrompt
   widget=BackingPanel
   properties {
    brush {
     bIsDynamicallyLoaded=false
     drawAs=Image
     imageType=FullColor
     mirroring=NoMirror
     resourceName=None
     resourceObject=/FortUILibrary/UIKit/Buttons/MaterialInstances/CTAButtons/MI_UIKit_Button_QuietRoundedCorners
     tiling=NoTile
     imageSize=<32 32>
     margin=<ltrb 0 0 0 0>
     outlineSettings {
      bUseBrushTransparency=false
      roundingType=HalfHeightRadius
      width=0
      color {colorUseRule=UseColor_Specified, specifiedColor=<rgba 0 0 0 0>}
      cornerRadii=<0 0 0 0>
     }
     tintColor {colorUseRule=UseColor_Specified, specifiedColor=<rgba 0.0123 0.0106 0.026 1>}
     uVRegion {bIsValid=false, max=<0 0>, min=<0 0>}
    }
    colorAndOpacity=<rgba 1 1 1 0.5>
   }
   slot {class=OverlaySlot, properties {horizontalAlignment=HAlign_Fill, verticalAlignment=VAlign_Fill}}
  }
  HorizontalBox_0 {
   widgetClassPath=&HorizontalBox
   parent=SongMenuPrompt
   widget=HorizontalBox_0
   slot {class=OverlaySlot, properties {verticalAlignment=VAlign_Center, padding=<ltrb 16 4 16 4>}}
  }
  ScaleBox_0 {
   widgetClassPath=&ScaleBox
   parent=HorizontalBox_0
   widget=ScaleBox_0
   properties {stretch=UserSpecified, userSpecifiedScale=0.75}
   slot {class=HorizontalBoxSlot, properties {padding=<ltrb 0 0 8 0>}}
  }
  CommonVisibilityWidgetLegacy_0 {
   widgetClassPath=&CommonVisibilityWidgetLegacy
   parent=ScaleBox_0
   widget=CommonVisibilityWidgetLegacy_0
   properties {bShowForTouch=false, visibleType=HitTestInvisible, padding=<ltrb 4 2 0 2>}
   slot {class=ScaleBoxSlot}
  }
  KeybindWidget {
   widgetClassPath=/FortUILibrary/Input/KeybindWidget
   bIsVariable=true
   parent=CommonVisibilityWidgetLegacy_0
   widget=KeybindWidget
   properties {
    bUseKeyboardTextureForKeys=true
    boundInputAction=/SparksSettings/Input/IA_OpenQuestPanel
    boundKeyFallback=Gamepad_Special_Left
    renderTransform {angle=0, scale=<0.9 0.9>, shear=<0 0>, translation=<0 0>}
   }
   slot {class=BorderSlot, properties {padding=<ltrb 4 2 0 2>}}
  }
  MenuLabel {
   widgetClassPath=&CommonTextBlock
   bIsVariable=true
   parent=HorizontalBox_0
   widget=MenuLabel
   properties {
    text=Game Menu
    font {
     bForceMonospaced=false
     bMaterialIsStencil=false
     fontMaterial=null
     fontObject=&HeadingNow
     letterSpacing=0
     monospacedWidth=1
     size=16.5
     skewAmount=0
     typefaceFontName=74 Regular
     outlineSettings {
      bApplyOutlineToDropShadows=false
      bMiteredCorners=false
      bSeparateFillAlpha=false
      outlineMaterial=&M-FontShadow
      outlineSize=2
      outlineColor=<rgba 0 0 0 1>
     }
    }
    shadowOffset=<0 0>
   }
   slot {class=HorizontalBoxSlot, properties {horizontalAlignment=HAlign_Left, verticalAlignment=VAlign_Center, padding=<ltrb 0 0 4 0>}}
  }
  SongMenuImage {
   widgetClassPath=&Image
   bIsVariable=true
   parent=HorizontalBox_0
   widget=SongMenuImage
   properties {
    visibility=SelfHitTestInvisible
    brush {
     bIsDynamicallyLoaded=false
     drawAs=Image
     imageType=NoImage
     mirroring=NoMirror
     resourceName=None
     resourceObject=/FMCore/Textures/T_Icon_SongMenu
     tiling=NoTile
     imageSize=<32 32>
     margin=<ltrb 0 0 0 0>
     outlineSettings {
      bUseBrushTransparency=false
      roundingType=HalfHeightRadius
      width=0
      color {colorUseRule=UseColor_Specified, specifiedColor=<rgba 0 0 0 0>}
      cornerRadii=<0 0 0 0>
     }
     tintColor {colorUseRule=UseColor_Specified, specifiedColor=<rgba 1 1 1 1>}
     uVRegion {bIsValid=false, max=<0 0>, min=<0 0>}
    }
   }
   slot {class=HorizontalBoxSlot}
  }
 ]
}
