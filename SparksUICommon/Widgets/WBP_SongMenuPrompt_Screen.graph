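# Screen-level widget that positions the song menu prompt at the top-left of the screen in the Sparks UI.
# An overlay insets a safe zone, padded on top and left only, around the WBP_SongMenuPrompt widget.
# A legacy CommonUI visibility wrapper hides the prompt for Android, iOS and touch, and the prompt itself is hit-test invisible.
# The event graph holds only unconnected default events, so the asset is layout only.
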WidgetBlueprint /SparksUICommon/Widgets/WBP_SongMenuPrompt_Screen
graphs [EventGraph {nodes [
   Event PreConstruct()
   Event Construct()
   Event Tick()
  ]}]
variables []
widgets {
 info {inheritedWidgetCount=0, parentClass=&UserWidget, rootWidgetClass=&Overlay, widgetCount=4}
 widgets [
  Overlay_1 {widgetClassPath=&Overlay, widget=Overlay_1}
  SZ_SongMenuPrompt {
   widgetClassPath=&SafeZone
   bIsVariable=true
   parent=Overlay_1
   widget=SZ_SongMenuPrompt
   properties {padBottom=false, padRight=false}
   slot {class=OverlaySlot, properties {padding=<ltrb 50 50 0 0>}}
  }
  CommonVisibilityWidgetLegacy_0 {
   widgetClassPath=&CommonVisibilityWidgetLegacy
   parent=SZ_SongMenuPrompt
   widget=CommonVisibilityWidgetLegacy_0
   properties {bShowForAndroid=false, bShowForIOS=false, bShowForTouch=false}
   slot {class=SafeZoneSlot}
  }
  WBP_SongMenuPrompt {
   widgetClassPath=&/WBP_SongMenuPrompt
   bIsVariable=true
   parent=CommonVisibilityWidgetLegacy_0
   widget=WBP_SongMenuPrompt
   properties {visibility=HitTestInvisible}
   slot {class=BorderSlot}
  }
 ]
}
