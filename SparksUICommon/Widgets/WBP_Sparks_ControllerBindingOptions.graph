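# Controller bindings options tab widget for Sparks UI, derived from FortTabControllerBindingsOptions_TempNew.
# Its only widget is an inherited SettingsPanel, which serves as the root.
# A SettingValueChanged custom event relays changes through the On Tab Setting Changed dispatcher without forwarding the value.
# The CenterOnTab event calls CenterWidget on that settings panel.
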WidgetBlueprint /SparksUICommon/Widgets/WBP_Sparks_ControllerBindingOptions
graphs [EventGraph {nodes [
   SettingValueChanged!() -> (OutputDelegate, then, Value:double=~) {is=title K2Node_CustomEvent}
   > Call On Tab Setting Changed!(execute:Exec, self:Fort Options Tab Object Reference) -> (then:Exec) {is=title K2Node_CallDelegate}
   Event CenterOnTab() {is=title K2Node_Event}
   > CenterWidget!(execute, ToCenter:Widget Object Reference=Get Settings_Panel() -> (Settings_Panel:Fort Settings Panel Object Reference)) -> (then)
  ]}]
variables []
widgets {
 info {
  inheritedWidgetCount=1
  parentClass=/Script/FortniteUI.FortTabControllerBindingsOptions_TempNew
  rootWidgetClass=/FortniteCoreUI/Settings/SettingsPanel.SettingsPanel_C
  widgetCount=1
 }
 widgets [Settings_Panel {widgetClassPath=/FortniteCoreUI/Settings/SettingsPanel, bInherited=true, bIsVariable=true, widget=Settings_Panel}]
}
