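# WBP_Sparks_Reticle is a reticle widget built on the native SparksReticle class, drawing a centered, non-hit-testable image with the M_UI_FM_Reticle material in an Overlay.
# On Construct, it collapses itself if a replay is playing and the replay reticle is not enabled; otherwise it is visible as HitTestInvisible.
# The visibility check runs only at construction, with no update logic in this graph.
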WidgetBlueprint /SparksUICommon/Widgets/WBP_Sparks_Reticle
graphs [EventGraph {nodes [
   Event Construct()
   > n1 = IsReplayReticleEnabled!(execute, self:Sparks Reticle Object Reference) -> (then, ReturnValue:bool)
   n2 = IsPlayingReplay!() -> (ReturnValue:bool)
   n3 = NOT Boolean(A=@n1)
   > SetVisibility(InVisibility=Select(Option 0=HitTestInvisible, Option 1=Collapsed, Index=AND Boolean(A=@n3, B=@n2)))
  ]}]
variables []
widgets {
 info {inheritedWidgetCount=0, parentClass=/Script/SparksUICommonRuntime.SparksReticle, rootWidgetClass=&Overlay, widgetCount=2}
 widgets [
  Overlay_Root {widgetClassPath=&Overlay, widget=Overlay_Root}
  Image_Dot {
   widgetClassPath=&Image
   bIsVariable=true
   parent=Overlay_Root
   widget=Image_Dot
   properties {
    visibility=HitTestInvisible
    brush {
     bIsDynamicallyLoaded=false
     drawAs=Image
     imageType=FullColor
     mirroring=NoMirror
     resourceName=None
     resourceObject=/FMCore/Materials/M_UI_FM_Reticle
     tiling=NoTile
     imageSize=<256 256>
     margin=<ltrb 0 0 0 0>
     outlineSettings {
      bUseBrushTransparency=false
      roundingType=HalfHeightRadius
      width=0
      color {colorUseRule=UseColor_Specified, specifiedColor=<rgba 0 0 0 0>}
      cornerRadii=<0 0 0 0>
     }
     tintColor {colorUseRule=UseColor_Specified, specifiedColor=<rgba 1 1 1 1>}
     uVRegion {bIsValid=false, max=<0 0>, min=<0 0>}
    }
   }
   slot {class=OverlaySlot, properties {horizontalAlignment=HAlign_Center, verticalAlignment=VAlign_Center}}
  }
 ]
}
