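# Parent pip widget that expands from a circle on a backplate into a wider pill showing a nested MOTD pip list when active.
# Initialize fills the child list only for multiple pips, computes the spacer's maximum width from pip size and spacing, and starts collapsed.
# SetActive plays Anim_Active in either direction, animating ActiveAnimProgress to lerp spacer width, then fading in the list.
# SetActiveChildPip forwards the selected index to the child list.
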WidgetBlueprint /SparksUICommon/Widgets/WBP_Pilgrim_Parent_Pip
animations [Anim_Active {
  binding_count=2
  length_seconds=0.2
  object=&self:Anim_Active
  sequence {bindings [
    WBP_MOTD_Pip_List {
     id=#1
     possessed_class=WBP_MOTD_Pip_List_C
     tracks [
      {
       type=MovieSceneByteTrack
       display_name=Visibility
       sections [{type=MovieSceneByteSection, end_frame=4, start_frame=0, channels [None {type=MovieSceneScriptingByteChannel, keys [[0,1],[2,0],[4,4]]}]}]
      }
      {
       type=MovieSceneFloatTrack
       display_name=Render Opacity
       sections [{type=MovieSceneFloatSection, end_frame=4, start_frame=0, channels [None {type=MovieSceneScriptingFloatChannel, keys [[2,0],[4,1]]}]}]
      }
     ]
    }
    WBP_Pilgrim_Pip_List_C_160 {
     id=#2
     possessed_class=WBP_Pilgrim_Parent_Pip_C
     tracks [{
       type=MovieSceneDoubleTrack
       display_name=Active Anim Progress
       sections [{type=MovieSceneDoubleSection, end_frame=2, start_frame=0, channels [None {type=MovieSceneScriptingDoubleChannel, keys [[0,0],[2,1]]}]}]
      }]
    }
   ]}
 }]
graphs [
 CalculateSpacerMaxWidth {nodes [
   n1 = CalculateSpacerMaxWidth!() -> (then, PipList:WBP MOTD Pip List Object Reference, MultiplePips:bool=~) {is=title K2Node_FunctionEntry}
   n2 = Get!(self:WBP MOTD Pip List Object Reference=@n1) -> (PipSize_X:double, PipSize_Y:double)
   n3 = float * float(A=2, B=@n2) {is=title K2Node_PromotableOperator}
   n4 = Get Count!(self:WBP MOTD Pip List Object Reference=@n1) -> (Count:int=0)
   n5 = float * float(A:int=@n4, B=@n2)
   n6 = Get MultiplePips() -> (MultiplePips:bool)
   n7 = Get!(self:Dynamic Entry Box Base Object Reference=Get PipBox!(self:WBP MOTD Pip List Object Reference=@n1) -> (PipBox:Dynamic Entry Box Object Reference)) -> (EntrySpacing_X:double, EntrySpacing_Y:double)
   n8 = float * float(A:double=@n7, B:int=int - int(A=@n4))
   > Set SpacerMaxWidth(SpacerMaxWidth:double=float + float(A=Select(Option 0:double=@n7, Option 1:double=@n8, Index=@n6) -> (ReturnValue:double), B=Select(Option 0=@n3, Option 1=@n5, Index=@n6))) -> (Output_Get:double)
  ]}
 EventGraph {nodes [
   n1 = Event PreConstruct()
   > Branch(Condition=@n1.IsDesignTime)
   > Initialize!(execute, self, PipCount:int=Get PipCount() -> (PipCount:int=0)) -> (then)
  ]}
 Initialize {nodes [
   n1 = Initialize!() -> (then, PipCount:int) {is=title K2Node_FunctionEntry}
   > n2 = Set MultiplePips(MultiplePips:bool="integer > integer"(A=@n1, B=1)) -> (Output_Get:bool)
   > n3 = Set PipCount(PipCount:int=Get PipCount() -> (PipCount=0)) -> (Output_Get:int=0)
   > SetPipsCount!(
    execute
    self:WBP MOTD Pip List Object Reference=Get WBP_MOTD_Pip_List() -> (WBP_MOTD_Pip_List:WBP MOTD Pip List Object Reference)
    Count:int=SelectInt(A=@n3, B=0, bPickA=@n2)
    Duration Value:double=0
    Anim Pip:bool=true
    Show Progress:bool=false
   ) -> (then)
   > CalculateSpacerMaxWidth!(execute, self, PipList:WBP MOTD Pip List Object Reference=Get WBP_MOTD_Pip_List(), MultiplePips:bool=Get MultiplePips()) -> (then) {is=title K2Node_CallFunction}
   n4 = Get Anim_Active() -> (Anim_Active:Widget Animation Object Reference)
   > PlayAnimation(InAnimation=@n4, StartAtTime=GetEndTime(self=@n4), PlayMode=Reverse)
   > Branch(Condition=Get MultiplePips())
   > Set SelectedColor(SelectedColor:Linear Color Structure=Get UnselectedColor() -> (UnselectedColor:Linear Color Structure)) -> (Output_Get:Linear Color Structure)
  ]}
 InitializeMaterial {nodes [InitializeMaterial() {is=title K2Node_FunctionEntry}]}
 SetActive {nodes [
   n1 = SetActive!() -> (then, IsActive:bool=~) {is=title K2Node_FunctionEntry}
   > n2 = Branch(Condition=@n1)
   n3 = Get Anim_Active()
   > PlayAnimationForward(InAnimation=@n3)
   PlayAnimationReverse(execute=@n2.else, InAnimation=@n3)
  ]}
 SetActiveAnimProgress {nodes [
   SetActiveAnimProgress!() -> (then, Value:double=~) {is=title K2Node_FunctionEntry}
   n1 = Get!(self:WBP MOTD Pip List Object Reference=Get WBP_MOTD_Pip_List()) -> (PipSize_X:double, PipSize_Y:double)
   > Set!(
    execute
    Size_X:double=Lerp(A=@n1, B=Get SpacerMaxWidth() -> (SpacerMaxWidth:double), Alpha=Get Value() -> (Value:double))
    Size_Y:double=@n1
    self:Spacer Object Reference=Get Spacer() -> (Spacer:Spacer Object Reference)
   ) -> (then, Output_Get:Vector 2D Structure)
  ]}
 SetActiveChildPip {nodes [
   n1 = SetActiveChildPip!() -> (then, Index:int) {is=title K2Node_FunctionEntry}
   > SetActivePip!(execute, self:WBP MOTD Pip List Object Reference=Get WBP_MOTD_Pip_List(), Index:int=@n1) -> (then)
  ]}
]
variables [
 ActiveAnimProgress {type=number}
 PipCount {type=integer}
 SelectedColor {type=LinearColor, default=<rgba 0.4125 0.7011 0.9823 1>}
 SpacerMaxWidth {type=number, default=32}
 UnselectedColor {type=LinearColor, default=<rgba 0.1785 0.1427 1 1>}
]
widgets {
 info {inheritedWidgetCount=0, parentClass=&UserWidget, rootWidgetClass=&GridPanel, widgetCount=4}
 widgets [
  GridPanel_0 {widgetClassPath=&GridPanel, widget=GridPanel_0, properties {bOverride_Cursor=true, columnFill [1], rowFill [1]}}
  WBP_UIKit_Backplate {
   widgetClassPath=&WBP_UIKit_Backplate
   bIsVariable=true
   parent=GridPanel_0
   widget=WBP_UIKit_Backplate
   properties {cornerRadius=Circle, colorAndOpacity=<rgba 0.1779 0.1413 1 1>}
   slot {class=GridSlot}
  }
  WBP_MOTD_Pip_List {
   widgetClassPath=/Motd/Widgets/WBP_MOTD_Pip_List
   bIsVariable=true
   parent=GridPanel_0
   widget=WBP_MOTD_Pip_List
   properties {visibility=Collapsed, pipSize=<14 14>}
   slot {class=GridSlot, properties {horizontalAlignment=HAlign_Center, verticalAlignment=VAlign_Center, padding=<ltrb 20 20 20 20>}}
  }
  Spacer {widgetClassPath=&Spacer, bIsVariable=true, parent=GridPanel_0, widget=Spacer, properties {size=<32 32>}, slot {class=GridSlot}}
 ]
}
