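# WBP_Pilgrim_Parent_Pip_List is a UI widget that shows a flat index as a position in a two-level row of parent and child pips.
# On PreConstruct it fills a Dynamic Entry Box with one parent pip per IndexMap entry, initialized with that child count, then selects index 0.
# SetCurrentIndex walks the cumulative IndexMap counts to find the parent and child for a raw index.
# SetActiveParentPip toggles active state only when the parent changes.
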WidgetBlueprint /SparksUICommon/Widgets/WBP_Pilgrim_Parent_Pip_List
graphs [
 EventGraph {nodes [
   Event PreConstruct()
   > FillEntryBox()
  ]}
 FillEntryBox {nodes [
   FillEntryBox() {is=title K2Node_FunctionEntry}
   > n1 = For Each Loop(Array:Array of Integers=Get IndexMap() -> (IndexMap:Array of Integers)) -> (Array Element:int)
   SetCurrentIndex!(execute=@n1.Completed, self, RawIndex:int=0) -> (then)
   n2 = Create Entry!(execute=@n1.LoopBody, self:Dynamic Entry Box Object Reference=Get DynamicEntryBox() -> (DynamicEntryBox:Dynamic Entry Box Object Reference)) -> (then, ReturnValue:User Widget Object Reference)
   > n3 = Cast To WBP_Pilgrim_Parent_Pip!(execute, Object=@n2) -> (then, CastFailed, AsWBP Pilgrim Parent Pip:WBP Pilgrim Parent Pip Object Reference) {is=title K2Node_DynamicCast}
   > Initialize!(execute, self:WBP Pilgrim Parent Pip Object Reference=@n3, PipCount:int=@n1) -> (then)
  ]}
 SetActiveParentPip {nodes [
   n1 = SetActiveParentPip!() -> (then, NewActiveParentPip:WBP Pilgrim Parent Pip Object Reference) {is=title K2Node_FunctionEntry}
   n2 = Get CurrentActiveParentPip() -> (CurrentActiveParentPip:WBP Pilgrim Parent Pip Object Reference)
   > n3 = Branch(Condition="Equal (Object)"(A=@n1, B=@n2))
   SetActive!(execute=@n3.else, self:WBP Pilgrim Parent Pip Object Reference=@n2, IsActive:bool=false) -> (then)
   > n4 = Set CurrentActiveParentPip(CurrentActiveParentPip:WBP Pilgrim Parent Pip Object Reference=Get NewActiveParentPip() -> (NewActiveParentPip:WBP Pilgrim Parent Pip Object Reference)) -> (Output_Get:WBP Pilgrim Parent Pip Object Reference)
   > SetActive(self=@n4, IsActive=true)
  ]}
 SetCurrentIndex {nodes [
   SetCurrentIndex!() -> (then, RawIndex:int) {is=title K2Node_FunctionEntry}
   > n1 = For Each Loop with Break!(Exec:Exec, Array:Array of Integers=Get IndexMap(), Break:Exec=@n7.then) -> (LoopBody:Exec, Array Element:int, Array Index:int, Completed:Exec)
   n2 = Get DynamicEntryBox()
   n3 = Cast To WBP_Pilgrim_Parent_Pip(
    execute=@n1.Completed
    Object="Get (a copy)"(
     Array:Array of User Widget Object References=GetAllEntries!(self:Dynamic Entry Box Base Object Reference=@n2) -> (ReturnValue:Array of User Widget Object References)
     Dimension 1=Get ParentIndex() -> (ParentIndex:int=0)
    ) -> (Output:User Widget Object Reference)
   )
   > SetActiveParentPip!(execute, self, NewActiveParentPip:WBP Pilgrim Parent Pip Object Reference=@n3) -> (then) {is=title K2Node_CallFunction}
   n4 = Get PipCount!(self:WBP Pilgrim Parent Pip Object Reference=@n3) -> (PipCount:int=0)
   n5 = Get ChildCounter() -> (ChildCounter:int=0)
   > SetActiveChildPip!(
    execute
    self:WBP Pilgrim Parent Pip Object Reference=@n3
    Index:int=int + int(A=int - int(A=@n4, B=@n5), B=Get RawIndex() -> (RawIndex:int=0))
   ) -> (then)
   Set ParentIndex(execute=@n1.LoopBody, ParentIndex:int=@n1.Array Index) -> (Output_Get:int=0)
   > n6 = Set ChildCounter(ChildCounter:int=int + int(A=@n1, B=Get ChildCounter() -> (ChildCounter=0))) -> (Output_Get:int=0)
   > n7 = Branch(Condition="integer > integer"(A=@n6, B=Get RawIndex() -> (RawIndex=0)))
  ]}
]
variables [
 CurrentActiveParentPip {type=WBP_Pilgrim_Parent_Pip_C, default=null}
 IndexMap {type="array<integer>", default [1,5,2]}
]
widgets {
 info {inheritedWidgetCount=0, parentClass=&UserWidget, rootWidgetClass=&DynamicEntryBox, widgetCount=1}
 widgets [DynamicEntryBox {
   widgetClassPath=&DynamicEntryBox
   bIsVariable=true
   widget=DynamicEntryBox
   properties {entryWidgetClass=&/WBP_Pilgrim_Parent_Pip.WBP_Pilgrim_Parent_Pip_C, numDesignerPreviewEntries=0, entrySpacing=<20 0>}
  }]
}
